(fn Close
  (Player|Stop (Variables|Default|GetSreamingVideo) true true))

(fn Create
  (params (bPriority :type Boolean) (bCreateAudioComponent :type Boolean) (bDebounce :type Boolean) (bCDNFailover :type Boolean :default true) (bIsBackgroundMusic :type Boolean))
  (bind _sreamingvideo (Variables|Default|GetSreamingVideo))
  (Utilities|FlowControl|Sequence
    (:then_0
      (Utilities|FlowControl|Sequence
        (:then_0
          (bind _returnvalue (Game|ConstructObjectfromClass "/Script/FortniteGame.FortBaseStreamingVideo"))
          (Variables|Default|SetSreamingVideo _returnvalue))
        (:then_1
          (bind _returnvalue_2 (Game|ConstructObjectfromClass "/Script/FortniteGame.FortStreamMediaSource"))
          (Variables|Default|SetStreamingMediaSource _returnvalue_2))
        (:then_2
          (bind _returnvalue_3 (Game|ConstructObjectfromClass "/Script/MediaAssets.MediaTexture"))
          (Variables|Default|SetMediaTexture _returnvalue_3))))
    (:then_1
      (bind _returnvalue_4 (Player|Init (Variables|Default|GetSreamingVideo) (Variables|Default|GetMediaTexture) 0 (Variables|Default|GetStreamingMediaSource) true))
      (LocalVariables|SetbResult _returnvalue_4))
    (:then_2
      (if (LocalVariables|GetbResult)
        (EventDispatchers|BindEventtoVideoonSuccess _sreamingvideo (EventDispatchers|CreateEvent :!event HandleSuccess))
        (EventDispatchers|BindEventtoVideoonTerminalError _sreamingvideo (EventDispatchers|CreateEvent :!event HandleTerminalError))
        (EventDispatchers|BindEventtoVideoonEndReached _sreamingvideo (EventDispatchers|CreateEvent :!event HandleEndReached))
        (EventDispatchers|BindEventtoVideoonClosed _sreamingvideo (EventDispatchers|CreateEvent :!event HandleClosed))))
    (:then_3
      (return (LocalVariables|GetbResult)))))

(fn HandleClosed
  (CallFunction|Close)
  (Default|CallOnClose self))

(fn HandleEndReached
  (Development|PrintText 0 (Utilities|Text|FormatText "NSLOCTEXT(\"[64B6258EF2F5727756AFBAA063D50422]\", \"22801A4B4205C485ED81389B72FF3C38\", \"End reasched with video {0}\")" self) true true "(R=1.000000,G=0.100000,B=0.100000,A=1.000000)" 5.0)
  (CallFunction|Close)
  (Default|CallOnClose self))

(fn HandleSuccess
  (CallFunction|Open))

(fn HandleTerminalError
  (params (Reason :type EBaseMediaTerminalErrorReason%20Enum :default "Unknown"))
  (Development|PrintText 0 (Utilities|Text|FormatText "NSLOCTEXT(\"[64B6258EF2F5727756AFBAA063D50422]\", \"80267AC34D9755931FDF9CAB91C1394D\", \"Terminal error with video {0}. Reason {1}\")" self (Utilities|Text|ToText Reason)) true true "(R=1.000000,G=0.100000,B=0.100000,A=1.000000)" 5.0)
  (Default|CallOnClose self))

(fn OnCloseSignature
  (params (Player :type BP%20Test%20Mini%20Player%20Object%20Reference)))

(fn Open
  (bind _sreamingvideo (Variables|Default|GetSreamingVideo))
  (bind _asfort_game_instance (Utilities|Casting|CastToFortGameInstance (Game|GetGameInstance self)))
  (Class|FortBaseStreamingVideo|SetPartyWatch _sreamingvideo (- (Math|DateTime|ToUnixTimestampDouble (Class|FortGameInstance|GetServerTime _asfort_game_instance)) (Fortnite|Sync|Time|GetServerWorldTimeSeconds self)) (Math|DateTime|ToUnixTimestampDouble (Class|FortGameInstance|GetServerTime _asfort_game_instance)) "Enabled" false)
  (Player|Open _sreamingvideo (struct :Tracks "(Audio=0,Caption=-1,Metadata=-1,Script=-1,Subtitle=-1,Text=-1,Video=0)" :TracksByLanguage "(Video=\"\",Audio=\"\",Subtitle=\"\",Caption=\"\")" :TrackSelection "UseTrackOptionIndices" :SeekTime "+00000000.00:00:00.000000000" :SeekTimeType "RelativeToStartTime" :PlayOnOpen "UseMediaPlayerSetting" :Loop "UseMediaPlayerSetting")))

(fn Start
  (params (MediaId :type String))
  (Player|Start (Variables|Default|GetSreamingVideo) MediaId (Variables|Default|GetMediaTexture)))
