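# Test subclass of B_FortVideoWidgetBase for the MediaFramework test map, playing a video by media ID with optional auto-play.
# Construct builds the media config, initializes the player and finds a watch-party start-time manager.
# Media callbacks broadcast state enums as stateful events on the global event router, which the widget also listens to.
# That listener logs states, plays pre-buffered video and passes server start/current times to SetPartyWatch for synced playback.
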WidgetBlueprint /Game/Athena/Maps/Test/MediaFramework/B_FortVideoWidgetBase_MediaFrameworkTest
graphs [
 EventGraph {nodes [
   DeactivateWidget()
   > SetVisibility(InVisibility=Collapsed)
   ActivateWidget(self=Self-Reference())
   > n1 = Branch(Condition=Get BP_AutoPlay() -> (BP_AutoPlay:bool))
   > Log It!(execute, self, InString:String="Setting Visibility to: Not Hit-Testable (Self Only)", bLogAsWarning:bool=false) -> (then)
   > SetVisibility(InVisibility=SelfHitTestInvisible)
   > Broadcast Stateful Event!(
    execute
    self:Gameplay Event Router Component Object Reference=GetGlobalEventRouter()
    EventData:Media Framework Test Video State Structure=Make Media Framework Test Video State!(Video_State_3_#1:MediaFrameworkTest_States Enum=NewEnumerator6) -> (MediaFrameworkTest_Video_State:Media Framework Test Video State Structure) {is=title K2Node_MakeStruct}
    EventContext:Object Reference
   ) -> (then, ReturnValue:bool)
   Log It(execute=@n1.else, InString=Setting Visibility to: Hidden, bLogAsWarning=false)
   > SetVisibility(InVisibility=Hidden)
   n2 = Force Deactivate Widget() {is=title K2Node_CustomEvent}
   Parent: On Deactivated(execute=[@n2.then, @n21.then]) {is=title K2Node_CallParentFunction}
   > Do Once()
   > Log It(InString=ON DEACTIVATED has been called, bLogAsWarning=false)
   Play PreBuffered Video() {is=title K2Node_CustomEvent}
   > n3 = Branch(Condition=Get BP_AutoPlay())
   > Log It(
    InString="PLAY PREBUFFERED VIDEO - Called with AutoPlay true, which isn't required, since it plays at the end of the Init flow."
    bLogAsWarning=true
   )
   Log It(execute=@n3.else, InString="PLAY PREBUFFERED VIDEO  - Called with AutoPlay false.", bLogAsWarning=true)
   > n4 = PlayPreBufferedVideo!(execute, self:Epic Media Video Widget Base Object Reference) -> (then, ReturnValue:bool)
   > Log It(InString=Append(A="Play Pre Buffering Video returned: ", B="To String (Boolean)"(InBool=@n4)), bLogAsWarning=false)
   > Branch(Condition=@n4)
   StartListeningForVideoEvents() {is=title K2Node_CustomEvent}
   > n5 = StartListeningToStatefulEvent!(
    execute:Exec
    Target:Gameplay Event Router Component Object Reference=GetGlobalEventRouter()
    EventType:Script Struct Object Reference=&/MediaFrameworkTest_Video_State
    Context:Object Reference
    EventBubblingRule:EEventBubblingRule Enum
   ) -> (
    then:Exec
    AsyncTaskProxy:Async Action Start Listening To Stateful Event Object Reference
    OnEventReceived:Exec
    OnSavedEventStateExists:Exec
    OnEventStateCleared:Exec
    Payload:Media Framework Test Video State Structure
   ) {is=title K2Node_StartListeningToStatefulEventAsyncAction}
   n6 = Switch on MediaFrameworkTest_States!(
    execute=[@n5.OnEventReceived, @n5.OnSavedEventStateExists]
    Selection:MediaFrameworkTest_States Enum=Break Media Framework Test Video State!(MediaFrameworkTest_Video_State:"Media Framework Test Video State Structure (by ref)"=@n5.Payload) -> (Video_State_3_#1:MediaFrameworkTest_States Enum) {is=title K2Node_BreakStruct}
   ) -> (
    NewEnumerator0:Exec
    NewEnumerator1:Exec
    NewEnumerator2:Exec
    NewEnumerator4:Exec
    NewEnumerator5:Exec
    NewEnumerator6:Exec
    NewEnumerator7:Exec
    NewEnumerator8:Exec
    NewEnumerator9:Exec
    NewEnumerator10:Exec
    NewEnumerator11:Exec
    NewEnumerator12:Exec
    NewEnumerator13:Exec
    NewEnumerator14:Exec
    NewEnumerator15:Exec
    NewEnumerator17:Exec
   ) {is=title K2Node_SwitchEnum}
   Log It(execute=@n6.NewEnumerator6, InString="Event  [Auto Play] -> (do nothing)", bLogAsWarning=false)
   Log It(execute=@n6.NewEnumerator14, InString="Event  [Video Playback Resumed]: (do nothing?)", bLogAsWarning=false)
   Log It(execute=@n6.NewEnumerator15, InString="Event  [Video Playback Suspended]: (do nothing)?", bLogAsWarning=false)
   Log It(execute=@n6.NewEnumerator9, InString="Event  [Prebuffer Play]: PlayPreBufferedVideo", bLogAsWarning=false)
   > PlayPreBufferedVideo()
   Log It(execute=@n6.NewEnumerator1, InString="Event  [Initializing] -> (do nothing)", bLogAsWarning=false)
   Log It(execute=@n6.NewEnumerator4, InString="Event  [Open Success] -> Begin Pre Buffering Video", bLogAsWarning=false)
   Log It(execute=@n6.NewEnumerator5, InString="Event  [Open Failure] -> (do nothing)", bLogAsWarning=false)
   Log It(execute=@n6.NewEnumerator7, InString="Event  [Prebuffer Start] -> (do nothing)", bLogAsWarning=false)
   Log It(execute=@n6.NewEnumerator8, InString="Event  [Prebuffer Completed] -> (do nothing)", bLogAsWarning=false)
   Log It(execute=@n6.NewEnumerator13, InString="Event  [Video Pre Ended]: (do nothing)?", bLogAsWarning=false)
   Log It(execute=@n6.NewEnumerator11, InString="Event  [Video Failure]: DeactivateWidget", bLogAsWarning=false)
   Log It(execute=@n6.NewEnumerator12, InString="Event  [Video Skipped]: DeactivateWidget", bLogAsWarning=false)
   Log It(execute=@n6.NewEnumerator17, InString="Event  [Video Seek Completed]: (do nothing?)", bLogAsWarning=false)
   > PlayPreBufferedVideo()
   Log It(execute=@n6.NewEnumerator2, InString="Event  [Metadata Success] -> (do nothing)", bLogAsWarning=false)
   n7 = Get WatchPartyServerStartTimeMgr() -> (WatchPartyServerStartTimeMgr:BP Watch Party Server Start Time Manager Object Reference)
   > Is Valid(InputObject=@n7)
   > n8 = Cast To FortGameInstance!(execute, Object=GetGameInstance!() -> (ReturnValue:Game Instance Object Reference)) -> (then, CastFailed, AsFort Game Instance:Fort Game Instance Object Reference) {is=title K2Node_DynamicCast}
   > n9 = Branch(Condition="float <= float"(A=Get ServerStartTime!(self:BP Watch Party Server Start Time Manager Object Reference=@n7) -> (ServerStartTime:double)) -> (ReturnValue=~))
   n10 = ToUnixTimestampDouble!(Time:"Date Time Structure (by ref)"=GetServerTime!(self:Fort Game Instance Object Reference=@n8) -> (ReturnValue:Date Time Structure)) -> (ReturnValue:double)
   > n11 = Set ServerStartTime!(execute, ServerStartTime:double=@n10, self:BP Watch Party Server Start Time Manager Object Reference=@n7) -> (then, Output_Get:double)
   n12 = GetBasePlayer!(execute=[@n11.then, @n9.else], self:Epic Media Video Widget Base Object Reference=Self-Reference()) -> (then, ReturnValue:Epic Base Streaming Video Object Reference)
   > n13 = Cast To FortBaseStreamingVideo!(execute, Object=@n12) -> (then, CastFailed, AsFort Base Streaming Video:Fort Base Streaming Video Object Reference) {is=title K2Node_DynamicCast}
   > SetPartyWatch!(
    execute
    self:Fort Base Streaming Video Object Reference=@n13
    InEventServerStartTime:double=Get ServerStartTime(self=Get WatchPartyServerStartTimeMgr())
    InEventServerCurrentTime:double=@n10
    InLoopOption:EMediaPlayerOptionBooleanOverride Enum=UseMediaPlayerSetting
    bAsLiveStream:bool=false
   ) -> (then)
   Log It(execute=@n6.NewEnumerator10, InString="Event  [Video Ended]: DeactivateWidget", bLogAsWarning=false)
   Init Media Player() {is=title K2Node_CustomEvent}
   > n14 = Construct Object from Class(Class=/Script/FortniteUI.FortMediaVideoWidgetConfig) -> (ReturnValue:Fort Media Video Widget Config Object Reference)
   > Set Media Config(Media Config:Fort Media Video Widget Config Object Reference=@n14) -> (Output_Get:Fort Media Video Widget Config Object Reference)
   > n15 = GetMediaId!(execute, self) -> (then, Result:String)
   > Set MediaID!(execute, MediaID:String=@n15, self:Epic Media Video Widget Config Object Reference=@n14) -> (then, Output_Get:String)
   > Set MediaName!(execute, MediaName:String=FortWidgetBase, self:Epic Media Video Widget Config Object Reference=@n14) -> (then, Output_Get:String)
   > Set bAutoPlay!(execute, bAutoPlay:bool=Get BP_AutoPlay(), self:Epic Media Video Widget Config Object Reference=@n14) -> (then, Output_Get:bool)
   > Set bAllowSkipping!(execute, bAllowSkipping:bool=true, self:Epic Media Video Widget Config Object Reference=@n14) -> (then, Output_Get:bool)
   > Set bHideControls!(execute, bHideControls:bool=false, self:Epic Media Video Widget Config Object Reference=@n14) -> (then, Output_Get:bool)
   > Set bLoop!(execute, bLoop:bool=false, self:Epic Media Video Widget Config Object Reference=@n14) -> (then, Output_Get:bool)
   > Set bNoAudio!(execute, bNoAudio:bool=false, self:Epic Media Video Widget Config Object Reference=@n14) -> (then, Output_Get:bool)
   > n16 = SetConfig!(
    execute
    self:Epic Media Video Widget Base Object Reference
    InConfig:Epic Media Video Widget Config Object Reference=Get Media Config() -> (Media Config:Fort Media Video Widget Config Object Reference)
   ) -> (then, ReturnValue:bool)
   > Branch(Condition=@n16)
   > Init!(execute, self:Epic Media Video Widget Base Object Reference) -> (then, ReturnValue:bool)
   n17 = Event PreConstruct()
   > Parent: PreConstruct(IsDesignTime=@n17.IsDesignTime)
   Event Construct()
   > Parent: Construct()
   > Log It(InString=Construct Called, bLogAsWarning=false)
   > n18 = Branch(Condition=Get BP_AutoPlay())
   StartListeningForVideoEvents(execute=[@n18.then, @n20.then]) {is=title K2Node_CallFunction}
   > Init Media Player() {is=title K2Node_CallFunction}
   > n19 = GetAllActorsOfClass!(
    execute
    ActorClass:Actor Class Reference=&/BP_WatchPartyServerStartTimeManager.BP_WatchPartyServerStartTimeManager_C
   ) -> (then, OutActors:Array of BP Watch Party Server Start Time Manager Object References)
   > Branch(Condition="integer > integer"(A=Length(TargetArray:Array of BP Watch Party Server Start Time Manager Object References=@n19)))
   > Set WatchPartyServerStartTimeMgr(WatchPartyServerStartTimeMgr:BP Watch Party Server Start Time Manager Object Reference="Get (a copy)"(Array:Array of BP Watch Party Server Start Time Manager Object References=@n19) -> (Output:BP Watch Party Server Start Time Manager Object Reference)) -> (Output_Get:BP Watch Party Server Start Time Manager Object Reference)
   n20 = SetVisibility(execute=@n18.else, InVisibility=Hidden)
   Event On Video Success() {is=title K2Node_Event}
   > Log It(InString=On Video Success called, bLogAsWarning=false)
   > Broadcast Stateful Event(self=GetGlobalEventRouter(), EventData=Make Media Framework Test Video State(Video_State_3_#1=NewEnumerator2))
   Event On Video Ended() {is=title K2Node_Event}
   > Log It(InString=On Video Ended called, bLogAsWarning=false)
   > Broadcast Stateful Event(self=GetGlobalEventRouter(), EventData=Make Media Framework Test Video State(Video_State_3_#1=NewEnumerator10))
   Event On Video Skipped() {is=title K2Node_Event}
   > Log It(InString=On Video Skipped called, bLogAsWarning=false)
   > Broadcast Stateful Event(self=GetGlobalEventRouter(), EventData=Make Media Framework Test Video State(Video_State_3_#1=NewEnumerator12))
   Event On Media Playback Resumed() {is=title K2Node_Event}
   > Log It(InString=On Media Playback Resumed called, bLogAsWarning=false)
   > Broadcast Stateful Event(self=GetGlobalEventRouter(), EventData=Make Media Framework Test Video State(Video_State_3_#1=NewEnumerator14))
   Event On Media Playback Suspended() {is=title K2Node_Event}
   > Log It(InString=On Media Playback Suspended called, bLogAsWarning=false)
   > Broadcast Stateful Event(self=GetGlobalEventRouter(), EventData=Make Media Framework Test Video State(Video_State_3_#1=NewEnumerator15))
   Event On Media Seek Completed() {is=title K2Node_Event}
   > Log It(InString=On Media Seek Completed called, bLogAsWarning=false)
   > Broadcast Stateful Event(self=GetGlobalEventRouter(), EventData=Make Media Framework Test Video State(Video_State_3_#1=NewEnumerator17))
   n21 = Event On Deactivated()
   Event On Media Opened Success() {is=title K2Node_Event}
   > Log It(InString=On Media Opened Success called, bLogAsWarning=false)
   > Broadcast Stateful Event(self=GetGlobalEventRouter(), EventData=Make Media Framework Test Video State(Video_State_3_#1=NewEnumerator4))
   Event PreEnded() {is=title K2Node_Event}
   > Parent: PreEnded() {is=title K2Node_CallParentFunction}
   > Log It(InString=PRE ENDED has been called, bLogAsWarning=false)
   Event On Media Opened Failed() {is=title K2Node_Event}
   > Log It(InString=On Media Opened Failed called, bLogAsWarning=false)
   > Broadcast Stateful Event(self=GetGlobalEventRouter(), EventData=Make Media Framework Test Video State(Video_State_3_#1=NewEnumerator5))
   Event On Media Pre Buffer Started() {is=title K2Node_Event}
   > Log It(InString=On Media Pre Buffer Started called, bLogAsWarning=false)
   > Broadcast Stateful Event(self=GetGlobalEventRouter(), EventData=Make Media Framework Test Video State(Video_State_3_#1=NewEnumerator7))
   Event On Media Pre Buffer Completed() {is=title K2Node_Event}
   > Log It(InString=On Media Pre Buffer Completed called, bLogAsWarning=false)
   > Broadcast Stateful Event(self=GetGlobalEventRouter(), EventData=Make Media Framework Test Video State(Video_State_3_#1=NewEnumerator8))
   Event On Media Metadata Success() {is=title K2Node_Event}
   > Log It(InString=On Media Metadata Success called, bLogAsWarning=false)
   > Broadcast Stateful Event(self=GetGlobalEventRouter(), EventData=Make Media Framework Test Video State(Video_State_3_#1=NewEnumerator2))
   Event On Video Terminal Error() {is=title K2Node_Event}
   > Log It(InString=On Video Terminal Error called, bLogAsWarning=false)
   > Broadcast Stateful Event(self=GetGlobalEventRouter(), EventData=Make Media Framework Test Video State(Video_State_3_#1=NewEnumerator11))
   n22 = Sequence!(execute) -> (then_0, then_1, then_2:Exec)
   > Do Once()
   > Broadcast Stateful Event(self=GetGlobalEventRouter(), EventData=Make Media Framework Test Video State(Video_State_3_#1=NewEnumerator13))
   > n23 = Cast To Athena_PlayerController!(
    execute
    Object=GetFirstLocalFortPlayerController!(bIncludeNonSpectators:bool=true, bIncludeSpectators:bool=false) -> (ReturnValue:Fort Player Controller Object Reference)
   ) -> (then, CastFailed, AsAthena Player Controller:Athena Player Controller Object Reference) {is=title K2Node_DynamicCast}
   > NotifyCinematicCameraSkipped(self:Fort Player Controller Athena Object Reference=@n23)
   > Log It(InString=PRE ENDED blueprint logic finishing, bLogAsWarning=false)
   n24 = GetEventRouter!(
    execute=@n22.then_1
    Scope:Object Class Reference=/Script/FortniteGame.FortPlayspace
    ContextActor:Actor Object Reference=GetFirstLocalFortPlayerController(bIncludeNonSpectators=true, bIncludeSpectators=false)
   ) -> (then, ReturnValue:Gameplay Event Router Component Object Reference)
   > Broadcast Event!(
    execute
    self:Gameplay Event Router Component Object Reference=@n24
    EventData:Media Framework Test Video State Structure=Make Media Framework Test Video State(Video_State_3_#1=NewEnumerator10)
    EventContext:Object Reference
   ) -> (then, ReturnValue:bool)
   > DeactivateWidget()
   > Log It(InString=ON DEACTIVATED blueprint logic finished, bLogAsWarning=false)
   Do Once()
   > n25 = Branch(Condition=Get Night Night After Video?() -> (Night Night After Video?:bool))
   > Log It(InString="Triggering Night Night, returning to front-end", bLogAsWarning=false)
   > Force Deactivate Widget() {is=title K2Node_CallFunction}
   > ClientSetupReturnToMainMenuForPlayers!(execute, bGoToNightNight:bool=true, MaxReturnTime:float=3) -> (then)
   Log It(execute=@n25.else, InString="Night Night disabled via cheat, still returning to front-end", bLogAsWarning=false)
   > Force Deactivate Widget()
   > ClientSetupReturnToMainMenuForPlayers(bGoToNightNight=false, MaxReturnTime=3)
   PlaylistDataReadyAsync!(execute:Exec) -> (
    then:Exec
    OnReady:Exec
    GameState:Fort Game State Athena Object Reference
    Playlist:Fort Playlist Object Reference
    PlaylistContextTags:"Gameplay Tag Container Structure (by ref)"
   ) {is=title K2Node_AsyncAction}
   SetVisibility(InVisibility=SelfHitTestInvisible)
   SetMediaPreEndTimeSeconds!(execute, self:Fort Video Widget Base Object Reference, InSeconds:float=2) -> (then)
   Play PreBuffered Video() {is=title K2Node_CallFunction}
   Trigger Night Night()
   Trigger Night Night() {is=title K2Node_CustomEvent}
  ]}
 GetMediaId {nodes [
   GetMediaId() {is=title K2Node_FunctionEntry}
   > Return Node!(execute, Result:String=Get MediaID() -> (MediaID:String))
  ]}
 Log It {nodes [
   n1 = Log It!() -> (then, InString:String, bLogAsWarning:bool=~) {is=title K2Node_FunctionEntry}
   > FortShippingLog!(
    execute
    WorldContextObject:Object Reference
    InString:String=Append(A="MediaFrameworkTest Video - B_FortVideoWidgetBase_MediaFrameworkTest] ", B=@n1)
    bLogAsWarning:bool=@n1.bLogAsWarning
   ) -> (then)
  ]}
 OnRep_PartyWatchServerStartTime {nodes [OnRep_PartyWatchServerStartTime() {is=title K2Node_FunctionEntry}]}
 OnRep_ServerStartTime {nodes [OnRep_ServerStartTime() {is=title K2Node_FunctionEntry}]}
]
variables [
 BP_AutoPlay {type=boolean}
 DeckBedAudioParams {type=LimeRock_AudioParameters_C, default=null}
 Delay Before Pre Buffering {type=ScalableFloat, default {registryType=None, value=0, curve {curveTable=null, rowName=None}}}
 Hide HUD Tag {type=GameplayTagContainer, default {gameplayTags [{tagName=Hud.Athena.HideOnCinematic}]}}
 Init Succeeded {type=boolean}
 Media Config {type=FortMediaVideoWidgetConfig, default=null}
 MediaID {type=string, default=cTXaWrpIhVLiOvjFaW}
 Night Night After Video? {type=boolean}
 Trigger Nite Nite on Video Init Failure {type=ScalableFloat, default {registryType=None, value=0, curve {curveTable=null, rowName=TriggerNiteNiteOnVideoFailure}}}
 Use Buffering Fix {type=ScalableFloat, default {registryType=None, value=0, curve {curveTable=null, rowName=None}}}
 Use New Video Flow {type=ScalableFloat, default {registryType=None, value=0, curve {curveTable=null, rowName=UseNewVideoFlow}}}
 WatchPartyServerStartTimeMgr {type=BP_WatchPartyServerStartTimeManager_C, default=null}
]
widgets {
 info {
  inheritedWidgetCount=0
  parentClass=/Game/UI/Foundation/Movie/B_FortVideoWidgetBase.B_FortVideoWidgetBase_C
  rootWidgetClass=null
  widgetCount=0
 }
 widgets []
}
