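# UI-domain test material for the Media Framework test content under the Athena test maps.
# It holds a single linear-color texture sample parameter named MediaTexture, which lets material instances supply a different texture.
# The parameter defaults to a texture from the Husk Deimos enemy folder, and the graph shows no other nodes or math.
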Material /Game/Athena/Maps/Test/MediaFramework/MT_TestMediaFramework
graph {
 properties {materialDomain=MD_UI, pixelDepthOffsetMode=PDOM_Legacy, usedShadingModels=Default Lit}
 nodes [TextureSampleParameter2D() {parameterName=MediaTexture, samplerType=SAMPLERTYPE_LinearColor, texture=/Game/Characters/Enemies/Husk_Deimos/Textures/Plane001NewStuffs}]
}
